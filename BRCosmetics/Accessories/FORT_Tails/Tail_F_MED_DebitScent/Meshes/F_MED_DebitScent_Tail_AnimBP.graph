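# Animation Blueprint for the DebitScent Fortnite tail cosmetic, posing and simulating the tail on the character.
# Chained bone-transform layers curl the tail, with slide and crouch offsets weighted by event-graph state, then per-bone world-space AnimDynamics adds spring motion.
# Skydive, glider, sprint and AnimDynamics variables are set, but the AnimGraph does not read them.
# On customization load, it hides the tail if backbling is a cape or long.
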AnimBlueprint /BRCosmetics/Accessories/FORT_Tails/Tail_F_MED_DebitScent/Meshes/F_MED_DebitScent_Tail_AnimBP
target_skeleton=/Game/Accessories/FORT_Tails/Common/Fortnite_Base_Tail_Skeleton
graphs [
 AnimGraph {nodes [
   n1 = "Transform (Modify) Bone - Bone: C_Tail_A_Base_Jnt"!(
    ComponentPose=Local To Component(LocalPose=Input Pose())
    bAlphaBoolEnabled:bool=Get Base_transform() -> (Base_transform:bool)
    Translation="0.5,-2.5,0.75"
    Rotation="0,-10,5"
    Scale="1,1,1"
   ) -> (Pose) {
    alphaInputType=Bool
    rotationMode=BMM_Additive
    rotationSpace=BCS_BoneSpace
    translationMode=BMM_Additive
    alphaBoolBlend {blendInTime=0.1, blendOption=QuadraticInOut, blendOutTime=0.1}
    boneToModify {boneName=C_Tail_A_Base_Jnt}
    rotation {roll=5, yaw=-10}
    translation=<0.5 -2.5 0.75>
   }
   n2 = Get SlideOffsetAlpha() -> (SlideOffsetAlpha:double)
   n3 = "Transform (Modify) Bone - Bone: C_Tail_A_Base_Jnt"!(ComponentPose=@n1, Alpha:float=@n2, Translation="0,0,0", Rotation="0,80,0", Scale="1,1,1") -> (Pose) {
    rotationMode=BMM_Additive
    rotationSpace=BCS_BoneSpace
    translationMode=BMM_Additive
    alphaBoolBlend {blendOption=Linear}
    alphaScaleBiasClamp {bInterpResult=true, interpSpeedDecreasing=40, interpSpeedIncreasing=40}
    boneToModify {boneName=C_Tail_A_Base_Jnt}
    rotation {yaw=80}
   }
   n4 = "Transform (Modify) Bone - Bone: C_Tail_A_1_Jnt"(ComponentPose=@n3, Alpha=@n2, Translation="0,0,0", Rotation="60,0,0", Scale="1,1,1") {
    rotationMode=BMM_Additive
    rotationSpace=BCS_BoneSpace
    translationMode=BMM_Additive
    alphaBoolBlend {blendOption=Linear}
    alphaScaleBiasClamp {bInterpResult=true, interpSpeedDecreasing=40, interpSpeedIncreasing=40}
    boneToModify {boneName=C_Tail_A_1_Jnt}
    rotation {pitch=60}
   }
   n5 = "Transform (Modify) Bone - Bone: C_Tail_A_2_Jnt"(ComponentPose=@n4, Alpha=@n2, Translation="0,0,0", Rotation="60,0,0", Scale="1,1,1") {
    rotationMode=BMM_Additive
    rotationSpace=BCS_BoneSpace
    translationMode=BMM_Additive
    alphaBoolBlend {blendOption=Linear}
    alphaScaleBiasClamp {bInterpResult=true, interpSpeedDecreasing=40, interpSpeedIncreasing=40}
    boneToModify {boneName=C_Tail_A_2_Jnt}
    rotation {pitch=60}
   }
   n6 = "Transform (Modify) Bone - Bone: C_Tail_A_3_Jnt"(ComponentPose=@n5, Alpha=@n2, Translation="0,0,0", Rotation="30,0,0", Scale="1,1,1") {
    rotationMode=BMM_Additive
    rotationSpace=BCS_BoneSpace
    translationMode=BMM_Additive
    alphaBoolBlend {blendOption=Linear}
    alphaScaleBiasClamp {bInterpResult=true, interpSpeedDecreasing=40, interpSpeedIncreasing=40}
    boneToModify {boneName=C_Tail_A_3_Jnt}
    rotation {pitch=30}
   }
   n7 = "Transform (Modify) Bone - Bone: C_Tail_A_Base_Jnt"(ComponentPose=@n6, Alpha=1, Translation="0,0,0", Rotation="0,-12,0", Scale="1,1,1") {
    rotationMode=BMM_Additive
    rotationSpace=BCS_BoneSpace
    translationMode=BMM_Additive
    alphaBoolBlend {blendOption=Linear}
    alphaScaleBiasClamp {bInterpResult=true, interpSpeedDecreasing=40, interpSpeedIncreasing=40}
    boneToModify {boneName=C_Tail_A_Base_Jnt}
    rotation {yaw=-12}
   }
   n8 = "Transform (Modify) Bone - Bone: C_Tail_A_Base_Jnt"(ComponentPose=@n7, Alpha=1, Translation="0,0,0", Rotation="0,20,0", Scale="1,1,1") {
    rotationMode=BMM_Replace
    rotationSpace=BCS_BoneSpace
    translationMode=BMM_Additive
    alphaBoolBlend {blendOption=Linear}
    alphaScaleBiasClamp {bInterpResult=true, interpSpeedIncreasing=4}
    boneToModify {boneName=C_Tail_A_Base_Jnt}
    rotation {yaw=20}
   }
   n9 = "Transform (Modify) Bone - Bone: C_Tail_A_1_Jnt"(ComponentPose=@n8, Alpha=1, Translation="0,0,0", Rotation="20,0,0", Scale="1,1,1") {
    rotationMode=BMM_Replace
    rotationSpace=BCS_BoneSpace
    translationMode=BMM_Additive
    alphaBoolBlend {blendOption=Linear}
    alphaScaleBiasClamp {bInterpResult=true, interpSpeedIncreasing=4}
    boneToModify {boneName=C_Tail_A_1_Jnt}
    rotation {pitch=20}
   }
   n10 = "Transform (Modify) Bone - Bone: C_Tail_A_2_Jnt"(ComponentPose=@n9, Alpha=1, Translation="0,0,0", Rotation="20,0,0", Scale="1,1,1") {
    rotationMode=BMM_Replace
    rotationSpace=BCS_BoneSpace
    translationMode=BMM_Additive
    alphaBoolBlend {blendOption=Linear}
    alphaScaleBiasClamp {bInterpResult=true, interpSpeedIncreasing=4}
    boneToModify {boneName=C_Tail_A_2_Jnt}
    rotation {pitch=20}
   }
   n11 = "Transform (Modify) Bone - Bone: C_Tail_A_3_Jnt"(ComponentPose=@n10, Alpha=1, Translation="0,0,0", Rotation="10,0,0", Scale="1,1,1") {
    rotationMode=BMM_Replace
    rotationSpace=BCS_BoneSpace
    translationMode=BMM_Additive
    alphaBoolBlend {blendOption=Linear}
    alphaScaleBiasClamp {bInterpResult=true, interpSpeedIncreasing=4}
    boneToModify {boneName=C_Tail_A_3_Jnt}
    rotation {pitch=10}
   }
   n12 = "Transform (Modify) Bone - Bone: C_Tail_A_4_Jnt"(ComponentPose=@n11, Alpha=1, Translation="0,0,0", Rotation="10,0,0", Scale="1,1,1") {
    rotationMode=BMM_Replace
    rotationSpace=BCS_BoneSpace
    translationMode=BMM_Additive
    alphaBoolBlend {blendOption=Linear}
    alphaScaleBiasClamp {bInterpResult=true, interpSpeedIncreasing=4}
    boneToModify {boneName=C_Tail_A_4_Jnt}
    rotation {pitch=10}
   }
   n13 = "Transform (Modify) Bone - Bone: C_Tail_A_5_Jnt"(ComponentPose=@n12, Alpha=1, Translation="0,0,0", Rotation="-10,0,0", Scale="1,1,1") {
    rotationMode=BMM_Replace
    rotationSpace=BCS_BoneSpace
    translationMode=BMM_Additive
    alphaBoolBlend {blendOption=Linear}
    alphaScaleBiasClamp {bInterpResult=true, interpSpeedIncreasing=4}
    boneToModify {boneName=C_Tail_A_5_Jnt}
    rotation {pitch=-10}
   }
   n14 = "Transform (Modify) Bone - Bone: C_Tail_A_6_Jnt"(ComponentPose=@n13, Alpha=1, Translation="0,0,0", Rotation="0,0,0", Scale="1,1,1") {
    rotationMode=BMM_Replace
    rotationSpace=BCS_BoneSpace
    translationMode=BMM_Additive
    alphaBoolBlend {blendOption=Linear}
    alphaScaleBiasClamp {bInterpResult=true, interpSpeedIncreasing=4}
    boneToModify {boneName=C_Tail_A_6_Jnt}
   }
   n15 = "Transform (Modify) Bone - Bone: C_Tail_A_7_Jnt"(ComponentPose=@n14, Alpha=1, Translation="0,0,0", Rotation="0,0,0", Scale="1,1,1") {
    rotationMode=BMM_Replace
    rotationSpace=BCS_BoneSpace
    translationMode=BMM_Additive
    alphaBoolBlend {blendOption=Linear}
    alphaScaleBiasClamp {bInterpResult=true, interpSpeedIncreasing=4}
    boneToModify {boneName=C_Tail_A_7_Jnt}
   }
   n16 = "Transform (Modify) Bone - Bone: C_Tail_A_8_Jnt"(ComponentPose=@n15, Alpha=1, Translation="0,0,0", Rotation="0,0,0", Scale="1,1,1") {
    rotationMode=BMM_Replace
    rotationSpace=BCS_BoneSpace
    translationMode=BMM_Additive
    alphaBoolBlend {blendOption=Linear}
    alphaScaleBiasClamp {bInterpResult=true, interpSpeedIncreasing=4}
    boneToModify {boneName=C_Tail_A_8_Jnt}
   }
   n17 = "Transform (Modify) Bone - Bone: C_Tail_A_9_Jnt"(ComponentPose=@n16, Alpha=1, Translation="0,0,0", Rotation="0,0,0", Scale="1,1,1") {
    rotationMode=BMM_Replace
    rotationSpace=BCS_BoneSpace
    translationMode=BMM_Additive
    alphaBoolBlend {blendOption=Linear}
    alphaScaleBiasClamp {bInterpResult=true, interpSpeedIncreasing=4}
    boneToModify {boneName=C_Tail_A_9_Jnt}
   }
   n18 = "Transform (Modify) Bone - Bone: C_Tail_A_Base_Jnt"(ComponentPose=@n17, Alpha=1, Translation="0,0,0", Rotation="0,20,0", Scale="1,1,1") {
    rotationMode=BMM_Replace
    rotationSpace=BCS_BoneSpace
    translationMode=BMM_Additive
    alphaBoolBlend {blendOption=Linear}
    alphaScaleBiasClamp {bInterpResult=true, interpSpeedDecreasing=30, interpSpeedIncreasing=4}
    boneToModify {boneName=C_Tail_A_Base_Jnt}
    rotation {yaw=20}
   }
   n19 = "Transform (Modify) Bone - Bone: C_Tail_A_Base_Jnt"(ComponentPose=@n18, Alpha=1, Translation="0,0,0", Rotation="0,6,0", Scale="1,1,1") {
    rotationMode=BMM_Replace
    rotationSpace=BCS_BoneSpace
    translationMode=BMM_Additive
    alphaBoolBlend {blendOption=Linear}
    alphaScaleBiasClamp {bInterpResult=true, interpSpeedDecreasing=30, interpSpeedIncreasing=4}
    boneToModify {boneName=C_Tail_A_Base_Jnt}
    rotation {yaw=6}
   }
   n20 = "Transform (Modify) Bone - Bone: C_Tail_A_Base_Jnt"(ComponentPose=@n19, Alpha=1, Translation="0,0,0", Rotation="0,20,0", Scale="1,1,1") {
    rotationMode=BMM_Replace
    rotationSpace=BCS_BoneSpace
    translationMode=BMM_Additive
    alphaBoolBlend {blendOption=Linear}
    alphaScaleBiasClamp {bInterpResult=true, interpSpeedDecreasing=30, interpSpeedIncreasing=4}
    boneToModify {boneName=C_Tail_A_Base_Jnt}
    rotation {yaw=20}
   }
   n21 = "Transform (Modify) Bone - Bone: C_Tail_A_1_Jnt"!(ComponentPose=@n20, bAlphaBoolEnabled:bool=True, Translation="0,0,0", Rotation="6,0,0", Scale="1,1,1") -> (Pose) {
    alphaInputType=Bool
    rotationMode=BMM_Additive
    rotationSpace=BCS_BoneSpace
    translationMode=BMM_Additive
    alphaBoolBlend {blendInTime=1, blendOption=Linear, blendOutTime=1}
    alphaScaleBiasClamp {bInterpResult=true, interpSpeedIncreasing=0.5}
    boneToModify {boneName=C_Tail_A_1_Jnt}
    rotation {pitch=6}
   }
   n22 = "Transform (Modify) Bone - Bone: C_Tail_A_2_Jnt"!(ComponentPose=@n21, bAlphaBoolEnabled:bool=True, Translation="0,0,0", Rotation="10,0,0", Scale="1,1,1") -> (Pose) {
    alphaInputType=Bool
    rotationMode=BMM_Additive
    rotationSpace=BCS_BoneSpace
    translationMode=BMM_Additive
    alphaBoolBlend {blendInTime=1, blendOption=Linear, blendOutTime=1}
    alphaScaleBiasClamp {bInterpResult=true, interpSpeedIncreasing=0.5}
    boneToModify {boneName=C_Tail_A_2_Jnt}
    rotation {pitch=10}
   }
   n23 = "Transform (Modify) Bone - Bone: C_Tail_A_3_Jnt"!(ComponentPose=@n22, bAlphaBoolEnabled:bool=True, Translation="0,0,0", Rotation="20,0,0", Scale="1,1,1") -> (Pose) {
    alphaInputType=Bool
    rotationMode=BMM_Additive
    rotationSpace=BCS_BoneSpace
    translationMode=BMM_Additive
    alphaBoolBlend {blendInTime=1, blendOption=Linear, blendOutTime=1}
    alphaScaleBiasClamp {bInterpResult=true, interpSpeedIncreasing=0.5}
    boneToModify {boneName=C_Tail_A_3_Jnt}
    rotation {pitch=20}
   }
   n24 = "Transform (Modify) Bone - Bone: C_Tail_A_4_Jnt"!(ComponentPose=@n23, bAlphaBoolEnabled:bool=True, Translation="0,0,0", Rotation="20,0,0", Scale="1,1,1") -> (Pose) {
    alphaInputType=Bool
    rotationMode=BMM_Additive
    rotationSpace=BCS_BoneSpace
    translationMode=BMM_Additive
    alphaBoolBlend {blendInTime=1, blendOption=Linear, blendOutTime=1}
    alphaScaleBiasClamp {bInterpResult=true, interpSpeedIncreasing=0.5}
    boneToModify {boneName=C_Tail_A_4_Jnt}
    rotation {pitch=20}
   }
   n25 = "Transform (Modify) Bone - Bone: C_Tail_A_5_Jnt"!(ComponentPose=@n24, bAlphaBoolEnabled:bool=True, Translation="0,0,0", Rotation="20,0,0", Scale="1,1,1") -> (Pose) {
    alphaInputType=Bool
    rotationMode=BMM_Additive
    rotationSpace=BCS_BoneSpace
    translationMode=BMM_Additive
    alphaBoolBlend {blendInTime=1, blendOption=Linear, blendOutTime=1}
    alphaScaleBiasClamp {bInterpResult=true, interpSpeedIncreasing=0.5}
    boneToModify {boneName=C_Tail_A_5_Jnt}
    rotation {pitch=20}
   }
   n26 = "Transform (Modify) Bone - Bone: C_Tail_A_6_Jnt"!(ComponentPose=@n25, bAlphaBoolEnabled:bool=True, Translation="0,0,0", Rotation="20,0,0", Scale="1,1,1") -> (Pose) {
    alphaInputType=Bool
    rotationMode=BMM_Additive
    rotationSpace=BCS_BoneSpace
    translationMode=BMM_Additive
    alphaBoolBlend {blendInTime=1, blendOption=Linear, blendOutTime=1}
    alphaScaleBiasClamp {bInterpResult=true, interpSpeedIncreasing=0.5}
    boneToModify {boneName=C_Tail_A_6_Jnt}
    rotation {pitch=20}
   }
   n27 = "Transform (Modify) Bone - Bone: C_Tail_A_7_Jnt"!(ComponentPose=@n26, bAlphaBoolEnabled:bool=True, Translation="0,0,0", Rotation="20,0,0", Scale="1,1,1") -> (Pose) {
    alphaInputType=Bool
    rotationMode=BMM_Additive
    rotationSpace=BCS_BoneSpace
    translationMode=BMM_Additive
    alphaBoolBlend {blendInTime=1, blendOption=Linear, blendOutTime=1}
    alphaScaleBiasClamp {bInterpResult=true, interpSpeedIncreasing=0.5}
    boneToModify {boneName=C_Tail_A_7_Jnt}
    rotation {pitch=20}
   }
   n28 = "Transform (Modify) Bone - Bone: C_Tail_A_8_Jnt"!(ComponentPose=@n27, bAlphaBoolEnabled:bool=True, Translation="0,0,0", Rotation="20,0,0", Scale="1,1,1") -> (Pose) {
    alphaInputType=Bool
    rotationMode=BMM_Additive
    rotationSpace=BCS_BoneSpace
    translationMode=BMM_Additive
    alphaBoolBlend {blendInTime=1, blendOption=Linear, blendOutTime=1}
    alphaScaleBiasClamp {bInterpResult=true, interpSpeedIncreasing=0.5}
    boneToModify {boneName=C_Tail_A_8_Jnt}
    rotation {pitch=20}
   }
   n29 = "Transform (Modify) Bone - Bone: C_Tail_A_Base_Jnt"(ComponentPose=@n28, Alpha=1, Translation="0,0,0", Rotation="0,10,0", Scale="1,1,1") {
    rotationMode=BMM_Additive
    rotationSpace=BCS_BoneSpace
    alphaBoolBlend {blendOption=Linear}
    boneToModify {boneName=C_Tail_A_Base_Jnt}
    rotation {yaw=10}
   }
   n30 = Get CrouchAlpha() -> (CrouchAlpha:double)
   n31 = "Transform (Modify) Bone - Bone: C_Tail_A_Base_Jnt"(ComponentPose=@n29, Alpha=@n30, Translation="0,0,0", Rotation="0,-10,0", Scale="1,1,1") {
    rotationMode=BMM_Replace
    rotationSpace=BCS_BoneSpace
    translationMode=BMM_Additive
    alphaBoolBlend {blendOption=Linear}
    alphaScaleBiasClamp {bInterpResult=true, interpSpeedIncreasing=20}
    boneToModify {boneName=C_Tail_A_Base_Jnt}
    rotation {yaw=-10}
   }
   n32 = "Transform (Modify) Bone - Bone: C_Tail_A_1_Jnt"!(ComponentPose=@n31, Alpha:float=@n30, Translation="0,0,0", Rotation="0,0,0", Scale="1,1,1") -> (Pose) {
    rotationMode=BMM_Replace
    rotationSpace=BCS_BoneSpace
    translationMode=BMM_Additive
    alphaBoolBlend {blendOption=Linear}
    alphaScaleBiasClamp {bInterpResult=true, interpSpeedIncreasing=20}
    boneToModify {boneName=C_Tail_A_1_Jnt}
   }
   n33 = "Transform (Modify) Bone - Bone: C_Tail_A_2_Jnt"!(ComponentPose=@n32, Alpha:float=@n30, Translation="0,0,0", Rotation="20,0,0", Scale="1,1,1") -> (Pose) {
    rotationMode=BMM_Replace
    rotationSpace=BCS_BoneSpace
    translationMode=BMM_Additive
    alphaBoolBlend {blendOption=Linear}
    alphaScaleBiasClamp {bInterpResult=true, interpSpeedIncreasing=20}
    boneToModify {boneName=C_Tail_A_2_Jnt}
    rotation {pitch=20}
   }
   n34 = "Transform (Modify) Bone - Bone: C_Tail_A_3_Jnt"!(ComponentPose=@n33, Alpha:float=@n30, Translation="0,0,0", Rotation="20,0,0", Scale="1,1,1") -> (Pose) {
    rotationMode=BMM_Replace
    rotationSpace=BCS_BoneSpace
    translationMode=BMM_Additive
    alphaBoolBlend {blendOption=Linear}
    alphaScaleBiasClamp {bInterpResult=true, interpSpeedIncreasing=20}
    boneToModify {boneName=C_Tail_A_3_Jnt}
    rotation {pitch=20}
   }
   n35 = "Transform (Modify) Bone - Bone: C_Tail_A_4_Jnt"!(ComponentPose=@n34, Alpha:float=@n30, Translation="0,0,0", Rotation="25,0,0", Scale="1,1,1") -> (Pose) {
    rotationMode=BMM_Replace
    rotationSpace=BCS_BoneSpace
    translationMode=BMM_Additive
    alphaBoolBlend {blendOption=Linear}
    alphaScaleBiasClamp {bInterpResult=true, interpSpeedIncreasing=20}
    boneToModify {boneName=C_Tail_A_4_Jnt}
    rotation {pitch=25}
   }
   n36 = "Transform (Modify) Bone - Bone: C_Tail_A_5_Jnt"!(ComponentPose=@n35, Alpha:float=@n30, Translation="0,0,0", Rotation="30,0,0", Scale="1,1,1") -> (Pose) {
    rotationMode=BMM_Replace
    rotationSpace=BCS_BoneSpace
    translationMode=BMM_Additive
    alphaBoolBlend {blendOption=Linear}
    alphaScaleBiasClamp {bInterpResult=true, interpSpeedIncreasing=20}
    boneToModify {boneName=C_Tail_A_5_Jnt}
    rotation {pitch=30}
   }
   n37 = "Transform (Modify) Bone - Bone: C_Tail_A_6_Jnt"!(ComponentPose=@n36, Alpha:float=@n30, Translation="0,0,0", Rotation="20,0,0", Scale="1,1,1") -> (Pose) {
    rotationMode=BMM_Replace
    rotationSpace=BCS_BoneSpace
    translationMode=BMM_Additive
    alphaBoolBlend {blendOption=Linear}
    alphaScaleBiasClamp {bInterpResult=true, interpSpeedIncreasing=20}
    boneToModify {boneName=C_Tail_A_6_Jnt}
    rotation {pitch=20}
   }
   n38 = "Transform (Modify) Bone - Bone: C_Tail_A_7_Jnt"!(ComponentPose=@n37, Alpha:float=@n30, Translation="0,0,0", Rotation="20,0,0", Scale="1,1,1") -> (Pose) {
    rotationMode=BMM_Replace
    rotationSpace=BCS_BoneSpace
    translationMode=BMM_Additive
    alphaBoolBlend {blendOption=Linear}
    alphaScaleBiasClamp {bInterpResult=true, interpSpeedIncreasing=20}
    boneToModify {boneName=C_Tail_A_7_Jnt}
    rotation {pitch=20}
   }
   n39 = "Transform (Modify) Bone - Bone: C_Tail_A_8_Jnt"!(ComponentPose=@n38, Alpha:float=@n30, Translation="0,0,0", Rotation="0,0,0", Scale="1,1,1") -> (Pose) {
    rotationMode=BMM_Replace
    rotationSpace=BCS_BoneSpace
    translationMode=BMM_Additive
    alphaBoolBlend {blendOption=Linear}
    alphaScaleBiasClamp {bInterpResult=true, interpSpeedIncreasing=20}
    boneToModify {boneName=C_Tail_A_8_Jnt}
   }
   n40 = Anim Dynamics - Bone: C_Tail_A_1_Jnt!(ComponentPose=@n39, Alpha=1) -> (Pose) {
    angularDampingOverride=0.999
    angularSpringConstant=700
    bAngularSpring=true
    bOverrideAngularDamping=true
    bUseGravityOverride=true
    gravityScale=0.35
    simulationSpace=World
    alphaBoolBlend {blendOption=Linear}
    alphaScaleBiasClamp {bInterpResult=true, interpSpeedDecreasing=4, interpSpeedIncreasing=4}
    boundBone {boneName=C_Tail_A_1_Jnt}
    componentAppliedLinearAccClamp=<1000 1000 1000>
    componentLinearAccScale=<1 1 1>
    componentLinearVelScale=<1 1 1>
    gravityOverride {z=-980}
    physicsBodyDefinitions [{
      boundBone {boneName=C_Tail_A_1_Jnt}
      boxExtents=<12 12 12>
      constraintSetup {angularLimitsMax=<0 3 25>, angularLimitsMin=<0 -50 -25>, angularTarget=<1 0 0>}
      localJointOffset=<0.5 0 0>
     }]
   }
   n41 = Anim Dynamics - Bone: C_Tail_A_2_Jnt!(ComponentPose=@n40, Alpha=1) -> (Pose) {
    angularDampingOverride=0.999
    angularSpringConstant=700
    bAngularSpring=true
    bOverrideAngularDamping=true
    bUseGravityOverride=true
    gravityScale=0.35
    simulationSpace=World
    alphaBoolBlend {blendOption=Linear}
    alphaScaleBiasClamp {bInterpResult=true, interpSpeedDecreasing=4, interpSpeedIncreasing=4}
    boundBone {boneName=C_Tail_A_2_Jnt}
    componentAppliedLinearAccClamp=<1000 1000 1000>
    componentLinearAccScale=<1 1 1>
    componentLinearVelScale=<1 1 1>
    gravityOverride {z=-980}
    physicsBodyDefinitions [{
      boundBone {boneName=C_Tail_A_2_Jnt}
      boxExtents=<12 12 12>
      constraintSetup {angularLimitsMax=<0 3 25>, angularLimitsMin=<0 -50 -25>, angularTarget=<1 0 0>}
      localJointOffset=<0.5 0 0>
     }]
   }
   n42 = Anim Dynamics - Bone: C_Tail_A_3_Jnt!(ComponentPose=@n41, Alpha=1) -> (Pose) {
    angularDampingOverride=0.999
    angularSpringConstant=700
    bAngularSpring=true
    bOverrideAngularDamping=true
    bUseGravityOverride=true
    gravityScale=0.35
    simulationSpace=World
    alphaBoolBlend {blendOption=Linear}
    alphaScaleBiasClamp {bInterpResult=true, interpSpeedDecreasing=4, interpSpeedIncreasing=4}
    boundBone {boneName=C_Tail_A_3_Jnt}
    componentAppliedLinearAccClamp=<1000 1000 1000>
    componentLinearAccScale=<1 1 1>
    componentLinearVelScale=<1 1 1>
    gravityOverride {z=-980}
    physicsBodyDefinitions [{
      boundBone {boneName=C_Tail_A_3_Jnt}
      boxExtents=<12 12 12>
      constraintSetup {angularLimitsMax=<0 3 25>, angularLimitsMin=<0 -50 -25>, angularTarget=<1 0 0>}
      localJointOffset=<0.5 0 0>
     }]
   }
   n43 = Anim Dynamics - Bone: C_Tail_A_4_Jnt!(ComponentPose=@n42, Alpha=1) -> (Pose) {
    angularDampingOverride=0.999
    angularSpringConstant=700
    bAngularSpring=true
    bOverrideAngularDamping=true
    bUseGravityOverride=true
    gravityScale=0.35
    simulationSpace=World
    alphaBoolBlend {blendOption=Linear}
    alphaScaleBiasClamp {bInterpResult=true, interpSpeedDecreasing=4, interpSpeedIncreasing=4}
    boundBone {boneName=C_Tail_A_4_Jnt}
    componentAppliedLinearAccClamp=<1000 1000 1000>
    componentLinearAccScale=<1 1 1>
    componentLinearVelScale=<1 1 1>
    gravityOverride {z=-980}
    physicsBodyDefinitions [{
      boundBone {boneName=C_Tail_A_4_Jnt}
      boxExtents=<12 12 12>
      constraintSetup {angularLimitsMax=<0 3 25>, angularLimitsMin=<0 -45 -25>, angularTarget=<1 0 0>}
      localJointOffset=<0.5 0 0>
     }]
   }
   n44 = Anim Dynamics - Bone: C_Tail_A_5_Jnt!(ComponentPose=@n43, Alpha=1) -> (Pose) {
    angularDampingOverride=0.999
    angularSpringConstant=700
    bAngularSpring=true
    bOverrideAngularDamping=true
    bUseGravityOverride=true
    gravityScale=0.35
    simulationSpace=World
    alphaBoolBlend {blendOption=Linear}
    alphaScaleBiasClamp {bInterpResult=true, interpSpeedDecreasing=4, interpSpeedIncreasing=4}
    boundBone {boneName=C_Tail_A_5_Jnt}
    componentAppliedLinearAccClamp=<1000 1000 1000>
    componentLinearAccScale=<1 1 1>
    componentLinearVelScale=<1 1 1>
    gravityOverride {z=-980}
    physicsBodyDefinitions [{
      boundBone {boneName=C_Tail_A_5_Jnt}
      boxExtents=<13 13 13>
      constraintSetup {angularLimitsMax=<0 3 25>, angularLimitsMin=<0 -40 -25>, angularTarget=<1 0 0>}
      localJointOffset=<0.5 0 0>
     }]
   }
   n45 = Anim Dynamics - Bone: C_Tail_A_6_Jnt!(ComponentPose=@n44, Alpha=1) -> (Pose) {
    angularDampingOverride=0.999
    angularSpringConstant=700
    bAngularSpring=true
    bOverrideAngularDamping=true
    bUseGravityOverride=true
    gravityScale=0.35
    simulationSpace=World
    alphaBoolBlend {blendOption=Linear}
    alphaScaleBiasClamp {bInterpResult=true, interpSpeedDecreasing=4, interpSpeedIncreasing=4}
    boundBone {boneName=C_Tail_A_6_Jnt}
    componentAppliedLinearAccClamp=<1000 1000 1000>
    componentLinearAccScale=<1 1 1>
    componentLinearVelScale=<1 1 1>
    gravityOverride {z=-980}
    physicsBodyDefinitions [{
      boundBone {boneName=C_Tail_A_6_Jnt}
      boxExtents=<14 14 14>
      constraintSetup {angularLimitsMax=<0 3 25>, angularLimitsMin=<0 -35 -25>, angularTarget=<1 0 0>}
      localJointOffset=<0.5 0 0>
     }]
   }
   n46 = Anim Dynamics - Bone: C_Tail_A_7_Jnt!(ComponentPose=@n45, Alpha=1) -> (Pose) {
    angularDampingOverride=0.999
    angularSpringConstant=700
    bAngularSpring=true
    bOverrideAngularDamping=true
    bUseGravityOverride=true
    gravityScale=0.35
    simulationSpace=World
    alphaBoolBlend {blendOption=Linear}
    alphaScaleBiasClamp {bInterpResult=true, interpSpeedDecreasing=4, interpSpeedIncreasing=4}
    boundBone {boneName=C_Tail_A_7_Jnt}
    componentAppliedLinearAccClamp=<1000 1000 1000>
    componentLinearAccScale=<1 1 1>
    componentLinearVelScale=<1 1 1>
    gravityOverride {z=-980}
    physicsBodyDefinitions [{
      boundBone {boneName=C_Tail_A_7_Jnt}
      boxExtents=<15 15 15>
      constraintSetup {angularLimitsMax=<0 3 25>, angularLimitsMin=<0 -30 -25>, angularTarget=<1 0 0>}
      localJointOffset=<0.5 0 0>
     }]
   }
   n47 = Anim Dynamics - Bone: C_Tail_A_8_Jnt!(ComponentPose=@n46, Alpha=1) -> (Pose) {
    angularDampingOverride=0.999
    angularSpringConstant=700
    bAngularSpring=true
    bOverrideAngularDamping=true
    bUseGravityOverride=true
    gravityScale=0.35
    simulationSpace=World
    alphaBoolBlend {blendOption=Linear}
    alphaScaleBiasClamp {bInterpResult=true, interpSpeedDecreasing=4, interpSpeedIncreasing=4}
    boundBone {boneName=C_Tail_A_8_Jnt}
    componentAppliedLinearAccClamp=<1000 1000 1000>
    componentLinearAccScale=<1 1 1>
    componentLinearVelScale=<1 1 1>
    gravityOverride {z=-980}
    physicsBodyDefinitions [{
      boundBone {boneName=C_Tail_A_8_Jnt}
      boxExtents=<16 16 16>
      constraintSetup {angularLimitsMax=<0 3 25>, angularLimitsMin=<0 -25 -25>, angularTarget=<1 0 0>}
      localJointOffset=<0.5 0 0>
     }]
   }
   n48 = Anim Dynamics - Bone: C_Tail_A_9_Jnt!(ComponentPose=@n47, Alpha=1) -> (Pose) {
    angularDampingOverride=0.999
    angularSpringConstant=700
    bAngularSpring=true
    bOverrideAngularDamping=true
    bUseGravityOverride=true
    gravityScale=0.35
    simulationSpace=World
    alphaBoolBlend {blendOption=Linear}
    alphaScaleBiasClamp {bInterpResult=true, interpSpeedDecreasing=4, interpSpeedIncreasing=4}
    boundBone {boneName=C_Tail_A_9_Jnt}
    componentAppliedLinearAccClamp=<1000 1000 1000>
    componentLinearAccScale=<1 1 1>
    componentLinearVelScale=<1 1 1>
    gravityOverride {z=-980}
    physicsBodyDefinitions [{
      boundBone {boneName=C_Tail_A_9_Jnt}
      boxExtents=<16 16 16>
      constraintSetup {angularLimitsMax=<0 3 25>, angularLimitsMin=<0 -25 -25>, angularTarget=<1 0 0>}
      localJointOffset=<0.5 0 0>
     }]
   }
   Output Pose(Result=Component To Local(ComponentPose=@n48))
  ]}
 EventGraph {nodes [
   Event BlueprintUpdateAnimation()
   > n1 = Sequence!(execute) -> (then_0, then_1, then_2:Exec, then_3:Exec)
   > n2 = Branch(Condition=Get bIsSkyDiving())
   > Set AnimDynamicsAlpha(AnimDynamicsAlpha:double=0) -> (Output_Get:double)
   > n3 = Sequence()
   > n4 = Branch(Condition=GliderItemDefinitionHasAnyMetaTag!(Tags:Array of Gameplay Tag Structures=Make Array("[0]"=(TagName="Cosmetics.MetaTags.CoconutShell")), bRequiresOpened:bool=true) -> (ReturnValue:bool))
   > Set matchingGliderDeployed(matchingGliderDeployed:bool=true) -> (Output_Get:bool)
   Set matchingGliderDeployed(execute=@n4.else, matchingGliderDeployed=false)
   n5 = Branch(execute=@n3.then_1, Condition=Get bIsSkydiveFloating())
   > Set SkydiveFloatingAlpha(SkydiveFloatingAlpha:double=1) -> (Output_Get:double)
   Set SkydiveFloatingAlpha(execute=@n5.else, SkydiveFloatingAlpha=0)
   n6 = Branch(execute=@n3.then_2, Condition=Get bIsSkydiveDiveMode())
   > Set DivingAlpha(DivingAlpha:double=1) -> (Output_Get:double)
   Set DivingAlpha(execute=@n6.else, DivingAlpha=0)
   n7 = Branch(execute=@n3.then_3, Condition=Get bIsParachuteOpened())
   > Set ParachutingAlpha(ParachutingAlpha:double=1) -> (Output_Get:double)
   Set ParachutingAlpha(execute=@n7.else, ParachutingAlpha=0)
   n8 = Branch(execute=@n2.else, Condition=Get IsOnZipline() -> (IsOnZipline:bool))
   > Set AnimDynamicsAlpha(AnimDynamicsAlpha=0)
   Set AnimDynamicsAlpha(execute=@n8.else, AnimDynamicsAlpha=1)
   > Set matchingGliderDeployed(matchingGliderDeployed=false)
   n9 = Branch(execute=@n1.then_1, Condition=OR Boolean(A=Get bIsJumping(), B=Get bIsFalling()))
   > Set AnimDynamicsAlpha(AnimDynamicsAlpha=0)
   Set AnimDynamicsAlpha(execute=@n9.else, AnimDynamicsAlpha=1)
   n10 = Get bIsSliding()
   n11 = Branch(execute=@n1.then_2, Condition=AND Boolean(A=Get bIsCrouching(), B=NOT Boolean(A=@n10)))
   > Set SlideOffsetAlpha(SlideOffsetAlpha:double=0.2) -> (Output_Get:double)
   > Set Base_transform(Base_transform:bool=false) -> (Output_Get:bool)
   > n12 = Branch(Condition=Get bIsMoving2D())
   > Set CrouchAlpha(CrouchAlpha:double=0) -> (Output_Get:double)
   Set CrouchAlpha(execute=@n12.else, CrouchAlpha=1)
   Set SlideOffsetAlpha(execute=@n11.else, SlideOffsetAlpha=1)
   > Set Base_transform(Base_transform=true)
   > Set CrouchAlpha(CrouchAlpha=0)
   n13 = Branch(execute=@n1.then_3, Condition=Get bIsSliding())
   > n14 = Branch(Condition=Get bIsKneeSliding())
   > Set SlideOffsetAlpha(SlideOffsetAlpha=0.75)
   > Set TacSprintAlpha(TacSprintAlpha:double=0) -> (Output_Get:double)
   Set SlideOffsetAlpha(execute=@n14.else, SlideOffsetAlpha=1)
   > Set TacSprintAlpha(TacSprintAlpha=0)
   Set SlideOffsetAlpha(execute=@n13.else, SlideOffsetAlpha=0)
   > n15 = Branch(Condition=OR Boolean(A=Get bIsTacticalSprinting(), B=Get IsOnZipline()))
   > Set TacSprintAlpha(TacSprintAlpha=1)
   Set TacSprintAlpha(execute=@n15.else, TacSprintAlpha=0)
   n16 = Event OnCharacterCustomizationLoaded()
   > n17 = Set Output(Output:Skeletal Mesh Component Budgeted Object Reference=Get Array Item with EFort Custom Part Type(Array=@n16.CustomMeshParts, PartType=MiscOrTail)) -> (Output_Get:Skeletal Mesh Component Budgeted Object Reference)
   n18 = CheckForBackbling()
   > n19 = Set isWearingCape(isWearingCape=OR Boolean(A=@n18.bIsBackblingACape, B=@n18.bIsBackblingLong))
   > n20 = Branch(Condition=@n19)
   > SetVisibility!(execute, self:Scene Component Object Reference=@n17, bNewVisibility:bool=false, bPropagateToChildren:bool=false) -> (then)
   SetVisibility(execute=@n20.else, self=@n17, bNewVisibility=true, bPropagateToChildren=false)
  ]}
]
variables [
 AnimDynamicsAlpha {type=number}
 Base_transform {type=boolean}
 CrouchAlpha {type=number, default=1}
 DivingAlpha {type=number}
 IsOnZipline {type=boolean}
 isWearingCape {type=boolean}
 matchingGliderDeployed {type=boolean}
 Output {type=SkeletalMeshComponentBudgeted}
 ParachutingAlpha {type=number}
 SkydiveFloatingAlpha {type=number}
 SlideOffsetAlpha {type=number}
 TacSprintAlpha {type=number}
]
